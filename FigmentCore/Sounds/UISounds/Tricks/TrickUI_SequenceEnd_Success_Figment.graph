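# Sound cue for the success sting at the end of a trick sequence in Figment's trick UI.
# A single wave player feeds the output directly with a Fortnite trick-success sound wave, with no randomization or modulation.
# It uses the gameplay music sound class, overrides concurrency to cap simultaneous instances, and has a send to the music audio-analysis submix.
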SoundCue /FigmentCore/Sounds/UISounds/Tricks/TrickUI_SequenceEnd_Success_Figment
root=WavePlayer_2
nodes [{id=WavePlayer_2, properties {soundWaveAssetPtr=/Game/Sounds/Fort_UI_Sounds/Gameplay/Tricks/TrickUI_SequenceEnd_Success_07}}]
properties {
 bOverrideConcurrency=true
 maxDistance=2097152
 soundClassObject=/Game/Sounds/FortSoundClassesAndModes/Fort_Music_Gameplay
 volumeMultiplier=1
 concurrencyOverrides {maxCount=2}
 soundSubmixSends [{minSendLevel=0, soundSubmix=&Fortnite_AudioAnalysis_Music}]
}
