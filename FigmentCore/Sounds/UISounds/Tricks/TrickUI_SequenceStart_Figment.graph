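# Sound cue for the Figment trick UI sequence start, layering a trick stinger into looping music on the gameplay music class.
# A root mixer plays a random NewTrick sound immediately, a quieter music intro after a short delay, then the loop layer after a longer delay.
# That layer pairs a looping button-hover sound with a music loop whose quality switch selects standard or OPT waves.
# It overrides concurrency and sends to the music audio-analysis submix.
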SoundCue /FigmentCore/Sounds/UISounds/Tricks/TrickUI_SequenceStart_Figment
root=Mixer_0
nodes [
 {id=Delay_0, children [WavePlayer_5], properties {delayMax=0.3, delayMin=0.3}}
 {id=Delay_1, children [Mixer_1], properties {delayMax=0.79, delayMin=0.79}}
 {id=Mixer_0, children [Random_0,Delay_0,Delay_1], properties {inputVolume [1,0.4,1]}}
 {id=Mixer_1, children [QualityLevel_0, WavePlayer_9], properties {inputVolume [1,1.75]}}
 {id=QualityLevel_0, children [WavePlayer_7,WavePlayer_7,WavePlayer_6,WavePlayer_6]}
 {id=Random_0, children [WavePlayer_0,WavePlayer_1,WavePlayer_3,WavePlayer_4], properties {weights [1,1,1,1]}}
 {id=WavePlayer_0, properties {soundWaveAssetPtr=&/TrickUI_NewTrick_06}}
 {id=WavePlayer_1, properties {soundWaveAssetPtr=&/TrickUI_NewTrick_05}}
 {id=WavePlayer_3, properties {soundWaveAssetPtr=&/TrickUI_NewTrick_04}}
 {id=WavePlayer_4, properties {soundWaveAssetPtr=&/TrickUI_NewTrick_03}}
 {id=WavePlayer_5, properties {soundWaveAssetPtr=&/TrickUI_SequenceMusicLoop_Intro_09}}
 {id=WavePlayer_6, properties {bLooping=true, soundWaveAssetPtr=&/TrickUI_SequenceMusicLoop_02_OPT_02}}
 {id=WavePlayer_7, properties {bLooping=true, soundWaveAssetPtr=&/TrickUI_SequenceMusicLoop_02}}
 {id=WavePlayer_9, properties {bLooping=true, soundWaveAssetPtr=&/UI_FN_Button_Hover_Figment}}
]
properties {
 bOverrideConcurrency=true
 maxDistance=2097152
 soundClassObject=/Game/Sounds/FortSoundClassesAndModes/Fort_Music_Gameplay
 volumeMultiplier=0.8
 concurrencyOverrides {maxCount=2}
 soundSubmixSends [{minSendLevel=0, soundSubmix=&Fortnite_AudioAnalysis_Music}]
}
